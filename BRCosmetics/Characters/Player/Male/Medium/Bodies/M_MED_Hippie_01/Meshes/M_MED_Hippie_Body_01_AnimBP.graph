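# Animation blueprint for the male medium hippie body cosmetic, adding secondary sway to its bell-bottom trousers.
# It copies the attached parent's pose, then chains spring-constrained AnimDynamics nodes on the left and right bell-bottom bones.
# Each update sets their shared external X/Y force from Perlin noise of the pawn's last Z position while skydiving, halved under an open parachute and otherwise zero.
# The Running_Around sequence player is unconnected.
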AnimBlueprint /BRCosmetics/Characters/Player/Male/Medium/Bodies/M_MED_Hippie_01/Meshes/M_MED_Hippie_Body_01_AnimBP
target_skeleton=&/M_MED_Hippie_Body_01_Skeleton
graphs [
 AnimGraph {nodes [
   n1 = Get falling_anim_dyn_up_force_vec3() -> (falling_anim_dyn_up_force_vec3:Vector)
   n2 = Anim Dynamics - Bone: dyn_bellbottom_l(ComponentPose=Local To Component(LocalPose=Copy Pose From Mesh() {bUseAttachedParent=true}), Alpha=0.32, ExternalForce=@n1) {
    bGravityOverrideInSimSpace=true
    bLinearSpring=true
    bOverrideLinearDamping=true
    lODThreshold=1
    linearDampingOverride=0.98
    linearSpringConstant=75
    numSolverIterationsPostUpdate=4
    numSolverIterationsPreUpdate=16
    alphaBoolBlend {blendOption=Linear}
    boundBone {boneName=dyn_bellbottom_l}
    physicsBodyDefinitions [{
      sphereCollisionRadius=0
      boundBone {boneName=dyn_bellbottom_l}
      boxExtents=<25 25 5>
      constraintSetup {angularLimitsMax=<5 10 10>, angularLimitsMin=<-5 -10 -10>, angularTarget=<0 0 0>, linearAxesMax=<8 8 0>, linearAxesMin=<-8 -8 0>}
      localJointOffset=<0 0 -10>
     }]
   }
   n3 = Anim Dynamics - Bone: dyn_bellbottom_r(ComponentPose=@n2, Alpha=0.32, ExternalForce=@n1) {
    bGravityOverrideInSimSpace=true
    bLinearSpring=true
    bOverrideLinearDamping=true
    lODThreshold=1
    linearDampingOverride=0.98
    linearSpringConstant=75
    numSolverIterationsPostUpdate=4
    numSolverIterationsPreUpdate=16
    alphaBoolBlend {blendOption=Linear}
    boundBone {boneName=dyn_bellbottom_r}
    physicsBodyDefinitions [{
      sphereCollisionRadius=0
      boundBone {boneName=dyn_bellbottom_r}
      boxExtents=<25 25 5>
      constraintSetup {angularLimitsMax=<5 10 10>, angularLimitsMin=<-5 -10 -10>, angularTarget=<0 0 0>, linearAxesMax=<8 8 0>, linearAxesMin=<-8 -8 0>}
      localJointOffset=<0 0 -10>
     }]
   }
   Output Pose(Result=Component To Local(ComponentPose=@n3)) {layerGroup=None}
   Sequence Player 'Running_Around'() {sequence=&/Running_Around}
  ]}
 EventGraph {nodes [
   Event BlueprintUpdateAnimation()
   n1 = Get PawnLastLocation() -> (PawnLastLocation:Vector)
   n2 = SelectFloat(
    A=Make Perlin Noise and Remap!(Value:float=BreakVector(InVec=@n1).Z, RangeOutMin:float=-25000, RangeOutMax:float=25000) -> (ReturnValue:float)
    B=0
    bPickA=Get bIsSkyDiving()
   )
   n3 = SelectFloat(A=float * float(A=@n2, B=0.5), B=@n2, bPickA=Get bIsParachuteOpened())
   > Set falling_anim_dyn_up_force_vec3(falling_anim_dyn_up_force_vec3:Vector=MakeVector(X=@n3, Y=@n3)) -> (Output_Get:Vector)
  ]}
]
variables [falling_anim_dyn_up_force_vec3 {type=Vector}]
