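# Physics asset for the medium male Hippie 01 player cosmetic body, giving simplified collision for the lower torso and upper legs only.
# Capsules cover the pelvis, two lower spine bones and both thighs, with an asymmetric setup: two shapes on the left thigh, one longer capsule on the right.
# Constraints link the thighs and first spine bone to the pelvis and chain the second spine bone above, with all swing and twist motion free.
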PhysicsAsset /BRCosmetics/Characters/Player/Male/Medium/Bodies/M_MED_Hippie_01/Meshes/M_MED_Hippie_Body_01_Physics
bodies {
 pelvis [{shapeType=Capsule, length=5.0664, radius=12.9863, center=<9.0316 -0.0002 0>, rotation=<rot 90 -0.0049 -0.0049>}]
 spine_01 [{shapeType=Capsule, length=10.8401, radius=12.6252, rotation=<rot 90 0 -0.5016>}]
 spine_02 [{shapeType=Capsule, length=5.5597, radius=12.1328, rotation=<rot 90 0 0>}]
 thigh_l [
  {shapeType=Capsule, length=10.6186, radius=10.1146, center=<-8.6442 0.865 1.4653>, rotation=<rot -88.8946 36.8845 -40.884>}
  {shapeType=Capsule, length=0, radius=8, center=<-0.0111 -2.2842 -0.0003>, rotation=<rot 0 0 90>}
 ]
 thigh_r [{shapeType=Capsule, length=31.6249, radius=8.844, center=<23.1726 -0.8702 -1.1238>, rotation=<rot -90 0.5968 -178.0469>}]
}
constraints [
 {
  bone1Name=thigh_l
  bone2Name=pelvis
  swing1LimitDegrees=45
  swing1Motion=Free
  swing2LimitDegrees=45
  swing2Motion=Free
  twistLimitDegrees=45
  twistMotion=Free
 }
 {
  bone1Name=thigh_r
  bone2Name=pelvis
  swing1LimitDegrees=45
  swing1Motion=Free
  swing2LimitDegrees=45
  swing2Motion=Free
  twistLimitDegrees=45
  twistMotion=Free
 }
 {
  bone1Name=spine_01
  bone2Name=pelvis
  swing1LimitDegrees=45
  swing1Motion=Free
  swing2LimitDegrees=45
  swing2Motion=Free
  twistLimitDegrees=45
  twistMotion=Free
 }
 {
  bone1Name=spine_02
  bone2Name=spine_01
  swing1LimitDegrees=45
  swing1Motion=Free
  swing2LimitDegrees=45
  swing2Motion=Free
  twistLimitDegrees=45
  twistMotion=Free
 }
]
