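# Post-style Animation Blueprint for the male Hippie head mesh that adds physics-driven secondary motion to hair and a hair band.
# It copies the pose from the attached parent mesh, then runs Anim Dynamics nodes in series, mostly single bones plus two short side chains.
# Angular springs and limits constrain each node, and only some hair nodes enable their spherical limits on spine_05.
# All logic lives in the AnimGraph.
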AnimBlueprint /BRCosmetics/Characters/Player/Male/Medium/Bodies/M_MED_Hippie_01/Meshes/M_MED_Hippie_Head_01_AnimBP
target_skeleton=&/M_MED_Hippie_Head_01_Skeleton
graphs [
 AnimGraph {nodes [
   n1 = Anim Dynamics - Bone: dyn_Hippie_hair_01(ComponentPose=Local To Component(LocalPose=Copy Pose From Mesh() {bUseAttachedParent=true}), Alpha=1, ExternalForce="0,0,0") {
    angularDampingOverride=0.98
    angularSpringConstant=1500
    bAngularSpring=true
    bGravityOverrideInSimSpace=true
    bOverrideAngularDamping=true
    numSolverIterationsPostUpdate=4
    numSolverIterationsPreUpdate=16
    alphaBoolBlend {blendOption=Linear}
    boundBone {boneName=dyn_Hippie_hair_01}
    physicsBodyDefinitions [{
      collisionType=CustomSphere
      sphereCollisionRadius=1
      boundBone {boneName=dyn_Hippie_hair_01}
      boxExtents=<20 45 45>
      constraintSetup {angularTargetAxis=AxisZ, angularLimitsMax=<5 20 0>, angularLimitsMin=<-5 -20 -20>, angularTarget=<0 0 1>}
      localJointOffset=<2 0 0>
     }]
    sphericalLimits [{limitRadius=12, drivingBone {boneName=neck_01}, sphereLocalOffset=<2.5 6 0>}]
   }
   n2 = Anim Dynamics - Bone: dyn_Hippie_hair_02(ComponentPose=@n1, Alpha=1, ExternalForce="0,0,0") {
    angularDampingOverride=0.98
    angularSpringConstant=1500
    bAngularSpring=true
    bGravityOverrideInSimSpace=true
    bOverrideAngularDamping=true
    numSolverIterationsPostUpdate=4
    numSolverIterationsPreUpdate=16
    alphaBoolBlend {blendOption=Linear}
    boundBone {boneName=dyn_Hippie_hair_02}
    physicsBodyDefinitions [{
      collisionType=CustomSphere
      sphereCollisionRadius=1
      boundBone {boneName=dyn_Hippie_hair_02}
      boxExtents=<20 45 45>
      constraintSetup {angularTargetAxis=AxisZ, angularLimitsMax=<5 20 0>, angularLimitsMin=<-5 -20 -20>, angularTarget=<0 0 1>}
      localJointOffset=<2 0 0>
     }]
    sphericalLimits [{limitRadius=12, drivingBone {boneName=neck_01}, sphereLocalOffset=<2.5 6 0>}]
   }
   n3 = Anim Dynamics - Bone: dyn_Hippie_hair_03(ComponentPose=@n2, Alpha=1, ExternalForce="0,0,0") {
    angularDampingOverride=0.98
    angularSpringConstant=100
    bAngularSpring=true
    bGravityOverrideInSimSpace=true
    bOverrideAngularDamping=true
    numSolverIterationsPostUpdate=4
    numSolverIterationsPreUpdate=16
    alphaBoolBlend {blendOption=Linear}
    boundBone {boneName=dyn_Hippie_hair_03}
    physicsBodyDefinitions [{
      collisionType=CustomSphere
      sphereCollisionRadius=1
      boundBone {boneName=dyn_Hippie_hair_03}
      boxExtents=<20 30 30>
      constraintSetup {angularTargetAxis=AxisZ, angularLimitsMax=<5 20 -5>, angularLimitsMin=<-5 -20 -60>, angularTarget=<0 0 1>}
      localJointOffset=<2 0 0>
     }]
    sphericalLimits [{limitRadius=24, drivingBone {boneName=spine_05}, sphereLocalOffset=<-4 11 0>}]
   }
   n4 = Anim Dynamics - Bone: dyn_Hippie_hair_04(ComponentPose=@n3, Alpha=1, ExternalForce="0,0,0") {
    angularDampingOverride=0.98
    angularSpringConstant=100
    bAngularSpring=true
    bGravityOverrideInSimSpace=true
    bOverrideAngularDamping=true
    bUseSphericalLimits=true
    numSolverIterationsPostUpdate=4
    numSolverIterationsPreUpdate=16
    alphaBoolBlend {blendOption=Linear}
    boundBone {boneName=dyn_Hippie_hair_04}
    physicsBodyDefinitions [{
      collisionType=CustomSphere
      sphereCollisionRadius=1
      boundBone {boneName=dyn_Hippie_hair_04}
      boxExtents=<20 30 30>
      constraintSetup {angularTargetAxis=AxisZ, angularLimitsMax=<5 20 0>, angularLimitsMin=<-5 -20 -60>, angularTarget=<0 0 1>}
      localJointOffset=<2 0 0>
     }]
    sphericalLimits [{limitRadius=6, drivingBone {boneName=spine_05}, sphereLocalOffset=<14 1 0>}]
   }
   n5 = Anim Dynamics - Bone: dyn_Hippie_hair_05(ComponentPose=@n4, Alpha=1, ExternalForce="0,0,0") {
    angularDampingOverride=0.98
    angularSpringConstant=100
    bAngularSpring=true
    bGravityOverrideInSimSpace=true
    bOverrideAngularDamping=true
    bUseSphericalLimits=true
    numSolverIterationsPostUpdate=4
    numSolverIterationsPreUpdate=16
    alphaBoolBlend {blendOption=Linear}
    boundBone {boneName=dyn_Hippie_hair_05}
    physicsBodyDefinitions [{
      collisionType=CustomSphere
      sphereCollisionRadius=1
      boundBone {boneName=dyn_Hippie_hair_05}
      boxExtents=<20 30 30>
      constraintSetup {angularTargetAxis=AxisZ, angularLimitsMax=<5 20 0>, angularLimitsMin=<-5 -20 -60>, angularTarget=<0 0 1>}
      localJointOffset=<2 0 0>
     }]
    sphericalLimits [{limitRadius=10, drivingBone {boneName=spine_05}, sphereLocalOffset=<4 2 1>}]
   }
   n6 = Anim Dynamics - Bone: dyn_Hippie_hair_01_l(ComponentPose=@n5, Alpha=1, ExternalForce="0,0,0") {
    angularDampingOverride=0.98
    angularSpringConstant=100
    bAngularSpring=true
    bGravityOverrideInSimSpace=true
    bOverrideAngularDamping=true
    numSolverIterationsPostUpdate=4
    numSolverIterationsPreUpdate=16
    alphaBoolBlend {blendOption=Linear}
    boundBone {boneName=dyn_Hippie_hair_01_l}
    physicsBodyDefinitions [{
      collisionType=CustomSphere
      sphereCollisionRadius=1.1
      boundBone {boneName=dyn_Hippie_hair_01_l}
      boxExtents=<20 30 30>
      constraintSetup {angularTargetAxis=AxisZ, angularLimitsMax=<5 20 0>, angularLimitsMin=<-5 -20 -20>, angularTarget=<0 0 1>}
      localJointOffset=<2 0 0>
     }]
    sphericalLimits [{limitRadius=24, drivingBone {boneName=spine_05}, sphereLocalOffset=<-4 11 -1>}]
   }
   n7 = "Anim Dynamics - Chain: dyn_Hippie_hair_02_l -> dyn_Hippie_hair_02_l1"(ComponentPose=@n6, Alpha=1, ExternalForce="0,0,0") {
    angularDampingOverride=0.98
    angularSpringConstant=100
    bAngularSpring=true
    bChain=true
    bGravityOverrideInSimSpace=true
    bOverrideAngularDamping=true
    bUseSphericalLimits=true
    numSolverIterationsPostUpdate=4
    numSolverIterationsPreUpdate=16
    alphaBoolBlend {blendOption=Linear}
    boundBone {boneName=dyn_Hippie_hair_02_l}
    chainEnd {boneName=dyn_Hippie_hair_02_l1}
    physicsBodyDefinitions [{
      collisionType=CustomSphere
      sphereCollisionRadius=1.1
      boundBone {boneName=dyn_Hippie_hair_02_l}
      boxExtents=<20 30 30>
      constraintSetup {angularTargetAxis=AxisZ, angularLimitsMax=<5 20 0>, angularLimitsMin=<-5 -20 -20>, angularTarget=<0 0 1>}
      localJointOffset=<2 0 0>
     }]
    sphericalLimits [{limitRadius=31, drivingBone {boneName=spine_05}, sphereLocalOffset=<-10 -15 0>}]
   }
   n8 = Anim Dynamics - Bone: dyn_Hippie_hair_03_l(ComponentPose=@n7, Alpha=1, ExternalForce="0,0,0") {
    angularDampingOverride=0.98
    angularSpringConstant=100
    bAngularSpring=true
    bGravityOverrideInSimSpace=true
    bOverrideAngularDamping=true
    numSolverIterationsPostUpdate=4
    numSolverIterationsPreUpdate=16
    alphaBoolBlend {blendOption=Linear}
    boundBone {boneName=dyn_Hippie_hair_03_l}
    physicsBodyDefinitions [{
      collisionType=CustomSphere
      sphereCollisionRadius=1.1
      boundBone {boneName=dyn_Hippie_hair_03_l}
      boxExtents=<20 30 30>
      constraintSetup {angularTargetAxis=AxisZ, angularLimitsMax=<5 20 0>, angularLimitsMin=<-5 -20 -20>, angularTarget=<0 0 1>}
      localJointOffset=<2 0 0>
     }]
    sphericalLimits [{limitRadius=24, drivingBone {boneName=spine_05}, sphereLocalOffset=<-4 11 1>}]
   }
   n9 = Anim Dynamics - Bone: dyn_Hippie_hair_01_r(ComponentPose=@n8, Alpha=1, ExternalForce="0,0,0") {
    angularDampingOverride=0.98
    angularSpringConstant=100
    bAngularSpring=true
    bGravityOverrideInSimSpace=true
    bOverrideAngularDamping=true
    numSolverIterationsPostUpdate=4
    numSolverIterationsPreUpdate=16
    alphaBoolBlend {blendOption=Linear}
    boundBone {boneName=dyn_Hippie_hair_01_r}
    physicsBodyDefinitions [{
      collisionType=CustomSphere
      sphereCollisionRadius=1.1
      boundBone {boneName=dyn_Hippie_hair_01_r}
      boxExtents=<20 30 30>
      constraintSetup {angularTargetAxis=AxisZ, angularLimitsMax=<5 20 20>, angularLimitsMin=<-5 -20 0>, angularTarget=<0 0 1>}
      localJointOffset=<-2 0 0>
     }]
    sphericalLimits [{limitRadius=24, drivingBone {boneName=spine_05}, sphereLocalOffset=<-4 11 -1>}]
   }
   n10 = Anim Dynamics - Bone: dyn_Hippie_hair_02_r(ComponentPose=@n9, Alpha=1, ExternalForce="0,0,0") {
    angularDampingOverride=0.98
    angularSpringConstant=100
    bAngularSpring=true
    bGravityOverrideInSimSpace=true
    bOverrideAngularDamping=true
    bUseSphericalLimits=true
    numSolverIterationsPostUpdate=4
    numSolverIterationsPreUpdate=16
    alphaBoolBlend {blendOption=Linear}
    boundBone {boneName=dyn_Hippie_hair_02_r}
    physicsBodyDefinitions [{
      collisionType=CustomSphere
      sphereCollisionRadius=1.1
      boundBone {boneName=dyn_Hippie_hair_02_r}
      boxExtents=<20 30 30>
      constraintSetup {angularTargetAxis=AxisZ, angularLimitsMax=<5 20 20>, angularLimitsMin=<-5 -20 0>, angularTarget=<0 0 1>}
      localJointOffset=<-2 0 0>
     }]
    sphericalLimits [{limitRadius=6, drivingBone {boneName=spine_05}, sphereLocalOffset=<8 1 6>}]
   }
   n11 = "Anim Dynamics - Chain: dyn_Hippie_hair_03_r -> dyn_Hippie_hair_03_r1"(ComponentPose=@n10, Alpha=1, ExternalForce="0,0,0") {
    angularDampingOverride=0.98
    angularSpringConstant=100
    bAngularSpring=true
    bChain=true
    bGravityOverrideInSimSpace=true
    bOverrideAngularDamping=true
    bUseSphericalLimits=true
    numSolverIterationsPostUpdate=4
    numSolverIterationsPreUpdate=16
    alphaBoolBlend {blendOption=Linear}
    boundBone {boneName=dyn_Hippie_hair_03_r}
    chainEnd {boneName=dyn_Hippie_hair_03_r1}
    physicsBodyDefinitions [{
      collisionType=CustomSphere
      sphereCollisionRadius=1.1
      boundBone {boneName=dyn_Hippie_hair_03_r}
      boxExtents=<20 30 30>
      constraintSetup {angularTargetAxis=AxisZ, angularLimitsMax=<5 20 20>, angularLimitsMin=<-5 -20 0>, angularTarget=<0 0 1>}
      localJointOffset=<-2 0 0>
     }]
    sphericalLimits [{limitRadius=31, drivingBone {boneName=spine_05}, sphereLocalOffset=<-10 -15 0>}]
   }
   n12 = Anim Dynamics - Bone: dyn_Hippie_hair_A_01_l(ComponentPose=@n11, Alpha=1, ExternalForce="0,0,0") {
    angularDampingOverride=0.98
    angularSpringConstant=100
    bAngularSpring=true
    bGravityOverrideInSimSpace=true
    bOverrideAngularDamping=true
    numSolverIterationsPostUpdate=4
    numSolverIterationsPreUpdate=16
    alphaBoolBlend {blendOption=Linear}
    boundBone {boneName=dyn_Hippie_hair_A_01_l}
    physicsBodyDefinitions [{
      collisionType=CustomSphere
      sphereCollisionRadius=1.1
      boundBone {boneName=dyn_Hippie_hair_A_01_l}
      boxExtents=<20 30 30>
      constraintSetup {angularTargetAxis=AxisZ, angularLimitsMax=<5 20 0>, angularLimitsMin=<-5 -20 -20>, angularTarget=<0 0 1>}
      localJointOffset=<2 0 0>
     }]
    sphericalLimits [{limitRadius=24, drivingBone {boneName=spine_05}, sphereLocalOffset=<-4 11 -1>}]
   }
   n13 = Anim Dynamics - Bone: dyn_Hippie_hair_A_02_l(ComponentPose=@n12, Alpha=1, ExternalForce="0,0,0") {
    angularDampingOverride=0.98
    angularSpringConstant=100
    bAngularSpring=true
    bGravityOverrideInSimSpace=true
    bOverrideAngularDamping=true
    numSolverIterationsPostUpdate=4
    numSolverIterationsPreUpdate=16
    alphaBoolBlend {blendOption=Linear}
    boundBone {boneName=dyn_Hippie_hair_A_02_l}
    physicsBodyDefinitions [{
      collisionType=CustomSphere
      sphereCollisionRadius=1.1
      boundBone {boneName=dyn_Hippie_hair_A_02_l}
      boxExtents=<20 30 30>
      constraintSetup {angularTargetAxis=AxisZ, angularLimitsMax=<5 20 0>, angularLimitsMin=<-5 -20 -20>, angularTarget=<0 0 1>}
      localJointOffset=<2 0 0>
     }]
    sphericalLimits [{limitRadius=24, drivingBone {boneName=spine_05}, sphereLocalOffset=<-4 11 1>}]
   }
   n14 = Anim Dynamics - Bone: dyn_Hippie_hair_band_03(ComponentPose=@n13, Alpha=1, ExternalForce="0,0,0") {
    angularDampingOverride=0.99
    angularSpringConstant=3000
    bAngularSpring=true
    bGravityOverrideInSimSpace=true
    bOverrideAngularDamping=true
    numSolverIterationsPostUpdate=4
    numSolverIterationsPreUpdate=16
    alphaBoolBlend {blendOption=Linear}
    boundBone {boneName=dyn_Hippie_hair_band_03}
    physicsBodyDefinitions [{
      sphereCollisionRadius=0
      boundBone {boneName=dyn_Hippie_hair_band_03}
      boxExtents=<19 19 13>
      constraintSetup {angularLimitsMax=<10 20 20>, angularLimitsMin=<-10 -20 -20>, angularTarget=<1 0 0>}
      localJointOffset=<0 0 -1.5>
     }]
   }
   Output Pose(Result=Component To Local(ComponentPose=@n14)) {layerGroup=None}
  ]}
 EventGraph {nodes [Event BlueprintUpdateAnimation()]}
]
variables []
